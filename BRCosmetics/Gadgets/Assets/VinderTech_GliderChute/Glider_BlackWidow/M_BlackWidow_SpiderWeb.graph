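# Translucent unlit, two-sided particle and trail material for the Black Widow glider's spider-web effect, tinted by time of day and particle color.
# Soft and hard web channels, weighted by dynamic parameters, use noise-distorted UVs with optional panning and world-position mapping.
# Opacity combines a trail or sphere mask, optional near-camera fade, particle alpha and, outside mobile, depth fade.
# An optional noise-driven waving offset is zeroed on mobile.
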Material /BRCosmetics/Gadgets/Assets/VinderTech_GliderChute/Glider_BlackWidow/M_BlackWidow_SpiderWeb
graph {
 properties {
  bUsedWithBeamTrails=true
  bUsedWithNiagaraSprites=true
  bUsedWithParticleSprites=true
  blendMode=BLEND_Translucent
  pixelDepthOffsetMode=PDOM_Legacy
  shadingModel=MSM_Unlit
  twoSided=true
  usedShadingModels=Unlit
 }
 nodes [
  n1 = CameraVectorWS()
  n2 = LinearInterpolate(Alpha=Normalize(VectorInput=ComponentMask(None=@n1) {b=true})) {constA=0.78}
  n3 = Constant() {r=1}
  n4 = MaterialFunctionCall(
   Color_Evening=VectorParameter() {parameterName=EveningTImeColor, defaultValue=<rgba 0.2835 0.3379 0.6979 1>}.RGB
   Color_Night=VectorParameter() {parameterName=NightTImeColor, defaultValue=<rgba 0.2835 0.3379 0.6979 1>}.RGB
  ) {materialFunction=&Color_Time_of_Day}
  n5 = ParticleColor()
  n6 = Multiply(A=@n5.RGB, B=Constant() {r=1})
  MaterialFunctionCall(Emissive=Multiply(A=StaticSwitchParameter(True=@n2, False=@n3) {defaultValue=true, parameterName=UseCameraVector}, B=Multiply(A=@n4.Result, B=@n6))) {materialFunction=&MF_FX_EmissiveAdjustments}
  n7 = TextureCoordinate() {vTiling=0.2}
  n8 = ComponentMask(None=TextureSample(UVs=Panner(Coordinate=@n7) {speedX=-0.8}) {texture=&T_NoisePack_01}.RGB) {g=true, r=true}
  n9 = ScalarParameter() {defaultValue=100, parameterName=UVdistortAmount}
  n10 = StaticSwitchParameter(True=ShadingPathSwitch(Default=Divide(A=@n8, B=@n9) {constB=8}, Mobile=Constant()) {desc=Migrated from FeatureLevelSwitch}, False=Constant()) {defaultValue=true, parameterName=UseUV_Distortion}
  n11 = DynamicParameter() {defaultValue=<rgba 1 1 2 1>, paramNames [TimeOffset, UV_Scale, SoftWebOpacity, HardWebOpacity]}
  n12 = Multiply(A=TextureCoordinate() {uTiling=0.9, vTiling=0.3}, B=@n11.UV_Scale)
  n13 = Add(A=@n11.TimeOffset, B=Multiply(A=Time(), B=ScalarParameter() {defaultValue=1, parameterName=PannerSpeed}))
  n14 = Panner(Coordinate=@n12, Time=@n13) {speedX=0.5}
  n15 = Panner(Coordinate=@n12, Time=@n11.TimeOffset) {speedX=0.5, speedY=0.5}
  n16 = TextureSampleParameter2D(UVs=Add(A=@n10, B=StaticSwitchParameter(True=@n14, False=@n15) {parameterName=PannerAnim})) {parameterName=Noise, texture=/Game/Effects/Fort_Effects/Textures/T_BlackWidow_SpiderWeb}
  n17 = SphereMask(
   A=WorldPosition().XYZ
   B=Add(A=ObjectPositionWS(), B=VectorParameter() {parameterName=InitialOffset, defaultValue=<rgba 0 0 -1000 1>}.RGB)
   Radius=Constant() {r=2550}
   Hardness=Constant()
  )
  n18 = Multiply(A=Multiply(A=@n17, B=@n17), B=ScalarParameter() {defaultValue=1, parameterName=World WebUV_U}) {constB=20}
  n19 = ComponentMask(None=TextureCoordinate()) {g=true}
  n20 = ScalarParameter() {defaultValue=1, parameterName=World WebUV_V}
  n21 = MaterialFunctionCall(Float2=Multiply(A=TextureCoordinate() {uTiling=2.2}, B=@n11.UV_Scale)) {materialFunction=&BreakOutFloat2Components}
  n22 = Divide(A=@n21.R)
  n23 = Sine(None=Multiply(A=@n21.G) {constA=2, constB=0.5})
  n24 = StaticSwitchParameter(
   True=AppendVector(A=Multiply(A=@n18) {constB=-1}, B=Multiply(A=@n19, B=@n20))
   False=AppendVector(A=LinearInterpolate(A=@n21.R, B=@n22, Alpha=@n23), B=@n21.G)
  ) {parameterName=World Position_UV}
  n25 = ScalarParameter() {defaultValue=0.5, parameterName=WebX_Speed}
  n26 = ScalarParameter() {parameterName=WebY_Speed}
  n27 = Panner(
   Coordinate=@n24
   Time=@n13
   Speed=StaticSwitchParameter(True=AppendVector(A=@n25, B=@n26), False=Constant2Vector() {r=0.5}) {parameterName=Skdive TrailB_speed}
  ) {speedX=0.5}
  n28 = MaterialFunctionCall(UVs=Panner(Coordinate=@n24, Time=@n11.TimeOffset) {speedY=0.5}, "Rotation Angle (0-1)"=Constant() {r=0.085}) {materialFunction=&CustomRotator}
  n29 = TextureSampleParameter2D(UVs=Add(A=@n10, B=StaticSwitchParameter(True=@n27, False=@n28.Rotated Values) {parameterName=PannerAnim})) {parameterName=Noise, texture=&T_BlackWidow_SpiderWeb}
  n30 = StaticSwitchParameter(True=@n29.B, False=@n29.B) {parameterName=SkydiveTrail_Web}
  n31 = MaterialFunctionCall(Float2=TextureCoordinate()) {materialFunction=&BreakOutFloat2Components}
  n32 = ScalarParameter() {defaultValue=0.5, parameterName=OpatcityMaskUbias}
  n33 = Constant() {r=0.5}
  n34 = ComponentMask(None=TextureCoordinate()) {g=true}
  n35 = Multiply(
   A=SphereMask(A=Power(Base=@n31.R, Exp=@n32) {constExponent=0.5}, B=@n33) {attenuationRadius=0.5, hardnessPercent=0}
   B=Saturate(None=SphereMask(A=@n34, B=@n33) {attenuationRadius=0.5, hardnessPercent=0})
  )
  n36 = MaterialFunctionCall(Float2=TextureCoordinate() {uTiling=0.6, vTiling=0.5}) {materialFunction=&BreakOutFloat2Components}
  n37 = Panner(Coordinate=AppendVector(A=Power(Base=@n36.R), B=@n36.G)) {speedX=-0.11}
  n38 = Multiply(A=@n35, B=Add(A=@n35, B=TextureSample(UVs=@n37) {texture=/Game/Global/Textures/Default/Noise/TankMoteNoise}.R))
  n39 = Multiply(A=@n35, B=TextureSample(UVs=@n37) {texture=&TankMoteNoise}.R)
  n40 = Multiply(A=MaterialFunctionCall(Float2=TextureCoordinate()) {materialFunction=&BreakOutFloat2Components}.R)
  n41 = OneMinus(None=Saturate(None=Multiply(A=@n40) {constB=5}))
  n42 = TextureCoordinate()
  n43 = Constant() {r=0.5}
  n44 = LinearInterpolate(Alpha=TextureSample() {samplerType=SAMPLERTYPE_Grayscale, texture=&LowResBlurredNoise}.R) {constA=0.2, constB=0.8}
  n45 = Multiply(
   A=Add(A=Multiply(A=@n16.R, B=@n11.SoftWebOpacity) {constB=2}, B=Multiply(A=@n11.HardWebOpacity, B=@n30))
   B=StaticSwitchParameter(True=LinearInterpolate(A=@n38, B=@n39, Alpha=@n41), False=SphereMask(A=@n42, B=@n43, Radius=@n44) {attenuationRadius=0.5, hardnessPercent=0}) {defaultValue=true, parameterName=UseTrailOpacity}
  )
  n46 = WorldPosition()
  n47 = CameraPositionWS()
  n48 = Constant() {r=100}
  n49 = Constant() {r=0.2}
  n50 = Multiply(A=@n45, B=OneMinus(None=SphereMask(A=@n46.XYZ, B=@n47, Radius=@n48, Hardness=@n49)))
  n51 = Multiply(
   A=@n5.A
   B=Multiply(A=StaticSwitchParameter(True=@n50, False=@n45) {parameterName=Camera_Mask}, B=ScalarParameter() {defaultValue=1, parameterName=Opacity})
  )
  n52 = DepthFade(Opacity=@n51, FadeDistance=ScalarParameter() {defaultValue=300, parameterName=DepthFade})
  ShadingPathSwitch(Default=@n52, Deferred=@n52, Forward=@n51, Mobile=@n51)
  n53 = Panner(
   Coordinate=TextureCoordinate() {coordinateIndex=1, uTiling=0.3, vTiling=0.3}
   Time=Add(A=DynamicParameter() {defaultValue=<rgba 1 1 2 1>, paramNames [TimeOffset, UV_Scale, SoftWebOpacity, HardWebOpacity]}.TimeOffset, B=Time())
   Speed=AppendVector(A=ScalarParameter() {parameterName=WavingSpeed}, B=Constant())
  )
  n54 = TextureSample(UVs=@n53) {samplerType=SAMPLERTYPE_Normal, texture=&T_Weap_Distort_N_02}
  n55 = Multiply(
   A=AppendVector(
    A=ComponentMask(None=@n54.RGB) {g=true, r=true}
    B=TextureSample(UVs=@n53) {texture=&T_Smooth_tile}.R
   )
   B=ScalarParameter() {parameterName=WavingAmount}
  ) {constB=1500}
  n56 = TextureCoordinate() {coordinateIndex=1, uTiling=0.3, vTiling=0.3}
  n57 = Saturate(None=Multiply(A=MaterialFunctionCall(Float2=@n56) {materialFunction=&BreakOutFloat2Components}.R) {constB=1.5})
  n58 = Constant()
  ShadingPathSwitch(Default=StaticSwitchParameter(True=Multiply(A=@n55, B=@n57), False=@n58) {parameterName=UseWavingAnim}, Mobile=@n58) {desc=Migrated from FeatureLevelSwitch}
  TextureSample() {texture=&TankMoteNoise}
 ]
}
